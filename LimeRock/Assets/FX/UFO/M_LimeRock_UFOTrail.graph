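# Translucent, two-sided lit material for UFO trail effects on Niagara ribbons.
# Two panning samples of a noise normal map are summed and scaled to distort UVs, breaking up an inverted U gradient with adjustable falloff.
# Power-shaped masks fade both ends along U and, through a sine applied twice, both edges along V.
# The gradient, masks and particle alpha multiply into one saturated trail mask.
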Material /LimeRock/Assets/FX/UFO/M_LimeRock_UFOTrail
graph {
 properties {bUsedWithNiagaraRibbons=true, blendMode=BLEND_Translucent, pixelDepthOffsetMode=PDOM_Legacy, twoSided=true, usedShadingModels=Default Lit}
 nodes [
  MaterialExpressionVertexColor_0 = ParticleColor()
  n1 = TextureCoordinate()
  n2 = Power(Base=Sine(None=ComponentMask(None=@n1) {g=true}) {period=2}, Exp=ScalarParameter() {defaultValue=2, parameterName=Mask_G_Power})
  n3 = ComponentMask(None=@n1) {r=true}
  n4 = Power(Base=@n3, Exp=ScalarParameter() {defaultValue=1, parameterName=Mask_R_FrontPower})
  n5 = Power(Base=OneMinus(None=@n3), Exp=ScalarParameter() {defaultValue=1, parameterName=Mask_R_BackPower})
  n6 = ScalarParameter() {defaultValue=0.5, group=Noise, parameterName=Noise_Amp}
  n7 = TextureCoordinate() {uTiling=0.25}
  n8 = AppendVector(
   A=ScalarParameter() {defaultValue=1, group=Noise, parameterName=Noise_Freq_U_01}
   B=ScalarParameter() {defaultValue=1, group=Noise, parameterName=Noise_Freq_V_01}
  )
  n9 = ScalarParameter() {defaultValue=-1, group=Noise, parameterName=Noise_Speed_U_01}
  n10 = ScalarParameter() {defaultValue=-0.05, group=Noise, parameterName=Noise_Speed_V_01}
  n11 = TextureSampleParameter2D(UVs=Panner(Coordinate=Multiply(A=@n7, B=@n8), Speed=AppendVector(A=@n9, B=@n10)) {speedX=-1, speedY=-0.05}) {group=Noise, parameterName=Noise_Texture_01, samplerType=SAMPLERTYPE_Normal, texture=&/T_LimeRock_Noise_N}
  n12 = TextureCoordinate() {uTiling=0.5}
  n13 = AppendVector(
   A=ScalarParameter() {defaultValue=1, group=Noise, parameterName=Noise_Freq_U_02}
   B=ScalarParameter() {defaultValue=1, group=Noise, parameterName=Noise_Freq_V_02}
  )
  n14 = ScalarParameter() {defaultValue=-1, group=Noise, parameterName=Noise_Speed_U_02}
  n15 = ScalarParameter() {defaultValue=0.2, group=Noise, parameterName=Noise_Speed_V_02}
  n16 = TextureSampleParameter2D(UVs=Panner(Coordinate=Multiply(A=@n12, B=@n13), Speed=AppendVector(A=@n14, B=@n15)) {speedX=-1, speedY=0.2}) {group=Noise, parameterName=Noise_Texture_02, samplerType=SAMPLERTYPE_Normal, texture=&T_LimeRock_Noise_N}
  n17 = ComponentMask(None=Add(A=@n11.RGB, B=@n16.RGB)) {g=true, r=true}
  n18 = ComponentMask(None=Add(A=TextureCoordinate(), B=Multiply(A=@n6, B=@n17))) {r=true}
  MaterialExpressionConstantClamp_1 = Saturate(None=Power(Base=OneMinus(None=@n18), Exp=ScalarParameter() {defaultValue=1, parameterName=GradientAmount}))
  n19 = Multiply(A=@MaterialExpressionConstantClamp_1, B=Multiply(A=@n2, B=Multiply(A=@n4, B=@n5)))
  MaterialExpressionConstantClamp_0 = Saturate(None=Multiply(A=@MaterialExpressionVertexColor_0.A, B=Multiply(A=@n2, B=@n19)))
 ]
}
