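# Niagara system for a UFO wind trail in LimeRock, drawn by a single ribbon emitter named Trail.
# Continuously spawned, very short-lived particles form a wide, semi-transparent white ribbon, with random material values written at spawn.
# User parameters Ribbon_SpawnRate and Ribbon_OpacityMult drive the spawn rate and alpha scale, exposing trail density and visibility to external control.
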NiagaraSystem /LimeRock/Assets/FX/UFO/NS_LimeRock_UFO_WindTrail
dependencies {usedDataInterfaces [], usedDynamicInputs [], usedModules [&SystemState]}
emitters {Trail {
  emitterUpdateScript [
   &EmitterState()
   &SpawnRate(SpawnRate=$User.Ribbon_SpawnRate, Use Spawn Probability=false)
  ]
  particleSpawnScript [
   &InitializeParticle(
    Color=<rgba 1 1 1 0.5>
    Lifetime=0.15
    Ribbon Width=900
    Ribbon Width Mode=::NewEnumerator1
    Sprite Size Mode=::NewEnumerator3
    Uniform Sprite Size=500
    UsePositionOffset=false
    Write Material Random=true
   )
   /SetVariables() {isSetParameters=true}
   &ApplyInitialForces()
  ]
  particleUpdateScript [
   &ParticleState()
   &ScaleColor(Scale Alpha=$User.Ribbon_OpacityMult, ScaleA=true, ScaleRGB=true)
   &SolveForcesAndVelocity(Clamp Velocity=false, Limit Acceleration=false)
  ]
  rendererClasses [&NiagaraRibbonRendererProperties]
  renderers [{rendererClass=&NiagaraRibbonRendererProperties}]
 }}
summary {
 systemName=NS_LimeRock_UFO_WindTrail
 emitters [{emitterName=Trail, rendererClasses [&NiagaraRibbonRendererProperties]}]
 userVariables [
  User.Ribbon_OpacityMult {type=NiagaraFloat, defaultValue=1, description=""}
  User.Ribbon_SpawnRate {type=NiagaraFloat, defaultValue=100, description=""}
 ]
}
